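; Puzzle manager for the coffee rig pipe valves that pays out loot once every valve is active.
; ValveTurned checks all valves until solved, then SolvedPuzzle latches the manager flag and writes false to each valve's PuzzleIsSolved.
; After a delay, authority spawns loot-tier pickups at the billboard and tosses them forward with random spread.
; Remotes play configured Niagara and sound effects immediately, then pirate chest effects after the delay.
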(event Custom|ValveTurned
  (bind _billboard (Variables|Default|GetBillboard))
  (bind _returnvalue (Transformation|GetWorldLocation _billboard))
  (bind _billboard_2 (Variables|Default|GetBillboard))
  (bind _returnvalue_2 (Transformation|GetForwardVector _billboard_2))
  (bind _toss_velocity (Variables|Default|GetToss%20Velocity))
  (bind _returnvalue_3 (Math|Conversions|ToVector(Float) _toss_velocity))
  (bind _returnvalue_4 (* _returnvalue_2 _returnvalue_3))
  (if (not (Variables|Default|GetPuzzleSolvedAtManager))
    (bind _allactive (CallFunction|CheckValveStates))
    (if _allactive
      (CallFunction|SolvedPuzzle)
      (Networking|SwitchHasAuthority
        (:Authority
          (Utilities|FlowControl|Delay 1.0)
          (Variables|Default|SetImpulse _returnvalue_4)
          (bind _returnvalue_5 (Inventory|SpawnPickupInWorldWithLootTier (Variables|Default|GetLoot%20Tier%20to%20Fire) (Transformation|GetWorldLocation _billboard_2) 0 0 true true _returnvalue_4))
          (Variables|Default|SetLoot _returnvalue_5)
          (for _array_element (Variables|Default|GetLoot)
            (bind _asfort_pickup_athena (Utilities|Casting|CastToFortPickupAthena _array_element))
            (Pickup|TossPickupinDirection _asfort_pickup_athena (+ (Math|Vector|MakeVector (* (Variables|Default|GetTossSpreadMultiplier) (Math|Random|RandomFloatinRange -4e+02 4e+02)) (* (Math|Random|RandomFloatinRange -4e+02 4e+02) (Variables|Default|GetTossSpreadMultiplier)) (* (Math|Random|RandomFloatinRange 1e+02 2.5e+02) (Variables|Default|GetTossSpreadMultiplier))) (* (Transformation|GetForwardVector (Variables|Default|GetBillboard)) (Variables|Default|GetToss%20Velocity))) self false false)))
        (:Remote
          (Niagara|SpawnSystematLocation (Variables|Default|GetParticleSystem) (+ _returnvalue (Variables|Default|GetEffectsOffset)))
          (Audio|SpawnSoundatLocation (Variables|Default|GetSoundCue) _returnvalue)
          (Utilities|FlowControl|Delay 1.0)
          (Audio|PlaySoundatLocation "/Game/Sounds/Foley_Loot/Containers/Pirate_Booty/Pirate_Booty_Chest_Open_Cue.Pirate_Booty_Chest_Open_Cue" (Transformation|GetActorLocation))
          (Niagara|SpawnSystematLocation "/Game/Effects/Fort_Effects/Gameplay/PirateChests/NS_Loot_Chest.NS_Loot_Chest" (Transformation|GetWorldLocation (Variables|Default|GetBillboard))))))))

(fn CheckValveStates
  (bind _array_element (Utilities|Array|ForEachLoopwithBreak (Variables|Default|GetAllPipeValvesCoffeeRig))
    (:LoopBody
      (if (not (Variables|/Hera_Map/Maps/Test/Fabricio/InteractiveProtos/BP_PipeValveInteractive_CoffeeRig.BP_PipeValveInteractive_CoffeeRig_C|GetValveActive _array_element))
        (LocalVariables|SetValveActiveLocal false)))
    (:Completed
      _))
  (return (LocalVariables|GetValveActiveLocal)))

(fn SolvedPuzzle
  (for _array_element (Variables|Default|GetAllPipeValvesCoffeeRig)
    (Variables|/Hera_Map/Maps/Test/Fabricio/InteractiveProtos/BP_PipeValveInteractive_CoffeeRig.BP_PipeValveInteractive_CoffeeRig_C|SetPuzzleIsSolved false _array_element))
  (Variables|Default|SetPuzzleSolvedAtManager true))

(fn ConstructionScript)
